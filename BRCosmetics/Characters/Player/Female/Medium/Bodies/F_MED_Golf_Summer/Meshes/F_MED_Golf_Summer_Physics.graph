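# Physics asset for the female medium Golf Summer player body cosmetic, covering the dynamic shoelace and zipper bones.
# Each shoe's lace root hangs from a tiny MSTR anchor body and branches into left and right two-bone capsule chains with tight angle limits.
# The zipper box hangs with angular limits from its own MSTR anchor, which is joined to a spine_05 torso capsule with free swing and twist.
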PhysicsAsset /BRCosmetics/Characters/Player/Female/Medium/Bodies/F_MED_Golf_Summer/Meshes/F_MED_Golf_Summer_Physics
bodies {
 dyn_shoelaceLeft_01_l [{shapeType=Capsule, length=1.5, radius=1, center=<0.8193 0 0>, rotation=<rot -90 1.2658 -1.2658>}]
 dyn_shoelaceLeft_01_r [{shapeType=Capsule, length=1.5, radius=1, center=<0.8193 0 0>, rotation=<rot -90 1.2658 -1.2658>}]
 dyn_shoelaceLeft_02_l [{shapeType=Capsule, length=1.5, radius=1, center=<0.8193 0 0>, rotation=<rot -90 1.2658 -1.2658>}]
 dyn_shoelaceLeft_02_r [{shapeType=Capsule, length=1.5, radius=1, center=<0.8193 0 0>, rotation=<rot -90 1.2658 -1.2658>}]
 dyn_shoelaceRight_01_l [{shapeType=Capsule, length=1.5, radius=1, center=<0.8193 0 0>, rotation=<rot -90 1.2658 -1.2658>}]
 dyn_shoelaceRight_01_r [{shapeType=Capsule, length=1.5, radius=1, center=<0.8193 0 0>, rotation=<rot -90 1.2658 -1.2658>}]
 dyn_shoelaceRight_02_l [{shapeType=Capsule, length=1.5, radius=1, center=<0.8193 0 0>, rotation=<rot -90 1.2658 -1.2658>}]
 dyn_shoelaceRight_02_r [{shapeType=Capsule, length=1.5, radius=1, center=<0.8193 0 0>, rotation=<rot -90 1.2658 -1.2658>}]
 dyn_shoelace_00_l [{shapeType=Capsule, length=1.5, radius=1, center=<0.8193 0 0>, rotation=<rot -90 1.2658 -1.2658>}]
 dyn_shoelace_00_l_MSTR [{shapeType=Capsule, length=0.01, radius=0.01}]
 dyn_shoelace_00_r [{shapeType=Capsule, length=1.5, radius=1, center=<0.8193 0 0>, rotation=<rot -90 1.2658 -1.2658>}]
 dyn_shoelace_00_r_MSTR [{shapeType=Capsule, length=0.01, radius=0.01}]
 dyn_zipper_01 [{shapeType=Box, extentX=1.51, extentY=1.01, extentZ=3.01, center=<1.3171 0 0.0357>, rotation=<rot 0 90 -88.2887>}]
 dyn_zipper_01_MSTR [{shapeType=Capsule, length=0.01, radius=0.01}]
 spine_05 [{shapeType=Capsule, length=16.9055, radius=9.1316, center=<0.3486 -1.1629 0>}]
}
constraints [
 {
  bone1Name=dyn_zipper_01_MSTR
  bone2Name=spine_05
  swing1LimitDegrees=45
  swing1Motion=Free
  swing2LimitDegrees=45
  swing2Motion=Free
  twistLimitDegrees=45
  twistMotion=Free
 }
 {bone1Name=dyn_zipper_01, bone2Name=dyn_zipper_01_MSTR, swing1LimitDegrees=10, swing2LimitDegrees=45, twistLimitDegrees=5}
 {bone1Name=dyn_shoelace_00_l, bone2Name=dyn_shoelace_00_l_MSTR, swing1LimitDegrees=5, swing2LimitDegrees=15, twistLimitDegrees=5}
 {bone1Name=dyn_shoelaceLeft_01_l, bone2Name=dyn_shoelace_00_l, swing1LimitDegrees=5, swing2LimitDegrees=25, twistLimitDegrees=5}
 {bone1Name=dyn_shoelaceLeft_02_l, bone2Name=dyn_shoelaceLeft_01_l, swing1LimitDegrees=5, swing2LimitDegrees=25, twistLimitDegrees=5}
 {bone1Name=dyn_shoelaceRight_01_l, bone2Name=dyn_shoelace_00_l, swing1LimitDegrees=5, swing2LimitDegrees=25, twistLimitDegrees=5}
 {bone1Name=dyn_shoelaceRight_02_l, bone2Name=dyn_shoelaceRight_01_l, swing1LimitDegrees=5, swing2LimitDegrees=25, twistLimitDegrees=5}
 {bone1Name=dyn_shoelace_00_r, bone2Name=dyn_shoelace_00_r_MSTR, swing1LimitDegrees=5, swing2LimitDegrees=15, twistLimitDegrees=5}
 {bone1Name=dyn_shoelaceRight_01_r, bone2Name=dyn_shoelace_00_r, swing1LimitDegrees=5, swing2LimitDegrees=15, twistLimitDegrees=5}
 {bone1Name=dyn_shoelaceRight_02_r, bone2Name=dyn_shoelaceRight_01_r, swing1LimitDegrees=5, swing2LimitDegrees=15, twistLimitDegrees=5}
 {bone1Name=dyn_shoelaceLeft_01_r, bone2Name=dyn_shoelace_00_r, swing1LimitDegrees=5, swing2LimitDegrees=15, twistLimitDegrees=5}
 {bone1Name=dyn_shoelaceLeft_02_r, bone2Name=dyn_shoelaceLeft_01_r, swing1LimitDegrees=5, swing2LimitDegrees=15, twistLimitDegrees=5}
]
