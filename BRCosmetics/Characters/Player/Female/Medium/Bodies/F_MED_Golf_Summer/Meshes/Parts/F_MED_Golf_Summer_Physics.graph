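# Physics asset for the F_MED_Golf_Summer female body part, providing capsule bodies and constraints for a ponytail.
# Four ponytail capsules form a chain rooted at a tiny master body joined to the head, with a tight first joint and wider swing but narrow twist further down.
# Head, neck and a three-capsule upper spine supply body volumes around the ponytail, linked by free angular constraints.
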PhysicsAsset /BRCosmetics/Characters/Player/Female/Medium/Bodies/F_MED_Golf_Summer/Meshes/Parts/F_MED_Golf_Summer_Physics
bodies {
 dyn_ponytail_01 [{shapeType=Capsule, length=3.6499, radius=2.894, center=<1.3435 0.0001 0>, rotation=<rot -90 0.0132 -0.0132>}]
 dyn_ponytail_01_MSTR [{shapeType=Capsule, length=0.01, radius=0.01}]
 dyn_ponytail_02 [{shapeType=Capsule, length=3.6499, radius=2.894, center=<1.3435 0.0001 0>, rotation=<rot -90 0.0132 -0.0132>}]
 dyn_ponytail_03 [{shapeType=Capsule, length=3.6499, radius=2.894, center=<4.1164 0.0001 0>, rotation=<rot -90 -0.0198 0.0198>}]
 dyn_ponytail_04 [{shapeType=Capsule, length=7.6499, radius=1.644, center=<3.7758 0.0002 0>, rotation=<rot -90 -0.0099 0.0099>}]
 head [{shapeType=Capsule, length=2.3975, radius=7.4819, center=<3.5246 0 -0.0001>, rotation=<rot -90 0.1583 -0.1583>}]
 neck_01 [{shapeType=Capsule, length=7.8975, radius=3.7319, center=<4.297 0 0>, rotation=<rot -90 174.9489 -174.9488>}]
 spine_05 [
  {shapeType=Capsule, length=19.6475, radius=6.9819, center=<6.07 0.6905 0>}
  {shapeType=Capsule, length=6.75, radius=5.25, center=<5.0812 -1.9158 -9.3676>, rotation=<rot -14.9999 0 0>}
  {shapeType=Capsule, length=6.75, radius=5.25, center=<5.7582 -1.9158 8.2685>, rotation=<rot 10.0001 0 0>}
 ]
}
constraints [
 {
  bone1Name=dyn_ponytail_01_MSTR
  bone2Name=head
  swing1LimitDegrees=45
  swing1Motion=Free
  swing2LimitDegrees=45
  swing2Motion=Free
  twistLimitDegrees=45
  twistMotion=Free
 }
 {bone1Name=dyn_ponytail_01, bone2Name=dyn_ponytail_01_MSTR, swing1LimitDegrees=5, swing2LimitDegrees=15, twistLimitDegrees=5}
 {bone1Name=dyn_ponytail_02, bone2Name=dyn_ponytail_01, swing1LimitDegrees=45, swing2LimitDegrees=45, twistLimitDegrees=5}
 {bone1Name=dyn_ponytail_03, bone2Name=dyn_ponytail_02, swing1LimitDegrees=45, swing2LimitDegrees=45, twistLimitDegrees=5}
 {bone1Name=dyn_ponytail_04, bone2Name=dyn_ponytail_03, swing1LimitDegrees=45, swing2LimitDegrees=45, twistLimitDegrees=5}
 {
  bone1Name=head
  bone2Name=neck_01
  swing1LimitDegrees=45
  swing1Motion=Free
  swing2LimitDegrees=45
  swing2Motion=Free
  twistLimitDegrees=45
  twistMotion=Free
 }
 {
  bone1Name=neck_01
  bone2Name=spine_05
  swing1LimitDegrees=45
  swing1Motion=Free
  swing2LimitDegrees=45
  swing2Motion=Free
  twistLimitDegrees=45
  twistMotion=Free
 }
]
